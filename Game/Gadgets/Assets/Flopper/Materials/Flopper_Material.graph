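# Default Lit material for the Flopper gadget, flagged for use on skeletal meshes with mesh deformers.
# It declares diffuse, spec and normal texture parameters, defaulting to textures from the LessEffectiveFlopper folder.
# The TestVariant static switch picks either plain diffuse or diffuse lerped toward the orange Test color, using the diffuse RGB as its own blend alpha.
# No connections to material outputs appear in the graph.
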Material /Game/Gadgets/Assets/Flopper/Materials/Flopper_Material
graph {
 properties {bUsedWithMeshDeformer=true, bUsedWithSkeletalMesh=true, pixelDepthOffsetMode=PDOM_Legacy, usedShadingModels=Default Lit}
 nodes [
  n1 = TextureSampleParameter2D() {parameterName=Diffuse, texture=/Game/Gadgets/Assets/LessEffectiveFlopper/Textures/T_Flopper_D}
  StaticSwitchParameter(
   True=LinearInterpolate(A=@n1.RGB, B=VectorParameter() {parameterName=Test, defaultValue=<rgba 1 0.3686 0 0>}.RGB, Alpha=@n1.RGB)
   False=@n1.RGB
  ) {parameterName=TestVariant}
  TextureSampleParameter2D() {parameterName=Spec Map, texture=/Game/Gadgets/Assets/LessEffectiveFlopper/Textures/T_Flopper_S}
  TextureSampleParameter2D() {parameterName=Normal, samplerType=SAMPLERTYPE_Normal, texture=/Game/Gadgets/Assets/LessEffectiveFlopper/Textures/T_Flopper_N}
 ]
}
